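# Physics asset for the SpadeStrut medium male body, set up for a left-side back cloth piece with collision capsules on the pelvis and left leg.
# Three dynamic cloth bones chain from a base bone locked angularly to the pelvis, with swing limits widening toward the tip and twist limits at zero.
# The thigh rotates freely against the pelvis and the calf is locked to the thigh; no other body parts have shapes.
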PhysicsAsset /BRCosmetics/Characters/Player/Male/Medium/Bodies/M_MED_SpadeStrut/Meshes/M_MED_SpadeStrut_Physics
bodies {
 base_cloth_back_01_lf [{shapeType=Capsule, length=0.505, radius=0.505, shapeName=base_cloth_back_01_lf_capsule, rotation=<rot 0 -90 0>}]
 calf_l [{shapeType=Capsule, length=39.5, radius=9.5, shapeName=calf_l_capsule, center=<-23.603 -1.9645 0>, rotation=<rot 90 0 0>}]
 dyn_cloth_back_01_lf [{shapeType=Capsule, length=3, radius=2, shapeName=dyn_cloth_back_01_lf_capsule, center=<2 0 0>, rotation=<rot 90 0 0>}]
 dyn_cloth_back_02_lf [{shapeType=Capsule, length=5, radius=1.5, shapeName=dyn_cloth_back_02_lf_capsule, center=<5 -1 0>, rotation=<rot -90 0 0>}]
 dyn_cloth_back_03_lf [{shapeType=Capsule, length=5, radius=1, shapeName=dyn_cloth_back_03_lf_capsule, center=<5 0 0>}]
 pelvis [{shapeType=Capsule, length=20, radius=10, center=<-7.7045 2.5197 -3.7362>}]
 thigh_l [{shapeType=Capsule, length=20, radius=7.4602, shapeName=thigh_l_capsule, center=<-26.3254 0.5393 0>, rotation=<rot 90 0 0>}]
}
constraints [
 {bone1Name=dyn_cloth_back_03_lf, bone2Name=dyn_cloth_back_02_lf, swing1LimitDegrees=40, swing2LimitDegrees=40, twistLimitDegrees=0}
 {bone1Name=dyn_cloth_back_02_lf, bone2Name=dyn_cloth_back_01_lf, swing1LimitDegrees=35, swing2LimitDegrees=30, twistLimitDegrees=0}
 {bone1Name=dyn_cloth_back_01_lf, bone2Name=base_cloth_back_01_lf, swing1LimitDegrees=5, swing2LimitDegrees=10, twistLimitDegrees=0}
 {
  bone1Name=thigh_l
  bone2Name=pelvis
  swing1LimitDegrees=45
  swing1Motion=Free
  swing2LimitDegrees=45
  swing2Motion=Free
  twistLimitDegrees=45
  twistMotion=Free
 }
 {
  bone1Name=calf_l
  bone2Name=thigh_l
  swing1LimitDegrees=45
  swing1Motion=Locked
  swing2LimitDegrees=45
  swing2Motion=Locked
  twistLimitDegrees=45
  twistMotion=Locked
 }
 {
  bone1Name=base_cloth_back_01_lf
  bone2Name=pelvis
  swing1LimitDegrees=45
  swing1Motion=Locked
  swing2LimitDegrees=45
  swing2Motion=Locked
  twistLimitDegrees=45
  twistMotion=Locked
 }
]
