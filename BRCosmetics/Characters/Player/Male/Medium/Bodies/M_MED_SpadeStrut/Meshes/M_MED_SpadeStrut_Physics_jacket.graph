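# Physics asset for the jacket of the M_MED_SpadeStrut male player body cosmetic, defining its collision bodies and constraints.
# Seven three-link dyn_jacket chains around the front, sides and back hang from base bones joined to spine_01 by free angular constraints.
# Chain links lock twist and limit swing, generally loosening toward the hem.
# Pelvis, spine and thigh capsules supply body shapes, with extra shapes on the right thigh and some front jacket links.
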PhysicsAsset /BRCosmetics/Characters/Player/Male/Medium/Bodies/M_MED_SpadeStrut/Meshes/M_MED_SpadeStrut_Physics_jacket
bodies {
 base_jacket_back_cr [{shapeType=Capsule, length=5, radius=1.5, shapeName=base_jacket_back_cr_capsule, rotation=<rot 0 -90 0>}]
 base_jacket_front_lf [{shapeType=Capsule, length=5, radius=1.5, shapeName=base_jacket_front_lf_capsule, rotation=<rot 0 -90 0>}]
 base_jacket_front_rt [{shapeType=Capsule, length=5, radius=1.5, shapeName=base_jacket_front_rt_capsule, rotation=<rot 0 -90 0>}]
 base_jacket_side_back_lf [{shapeType=Capsule, length=5, radius=1.5, shapeName=base_jacket_side_back_lf_capsule, rotation=<rot 0 -90 0>}]
 base_jacket_side_back_rt [{shapeType=Capsule, length=5, radius=1.5, shapeName=base_jacket_side_back_rt_capsule, rotation=<rot 0 -90 0>}]
 base_jacket_side_front_lf [{shapeType=Capsule, length=5, radius=1.5, shapeName=base_jacket_side_front_lf_capsule, rotation=<rot 0 -90 0>}]
 base_jacket_side_front_rt [{shapeType=Capsule, length=5, radius=1.5, shapeName=base_jacket_side_front_rt_capsule, rotation=<rot 0 -90 0>}]
 dyn_jacket_back_01_cr [{shapeType=Capsule, length=5, radius=1.5, shapeName=dyn_jacket_back_01_cr_capsule, center=<1 0 0>, rotation=<rot 90 0 0>}]
 dyn_jacket_back_02_cr [{shapeType=Capsule, length=5, radius=1.5, shapeName=dyn_jacket_back_02_cr_capsule, rotation=<rot 90 0 0>}]
 dyn_jacket_back_03_cr [{shapeType=Capsule, length=5, radius=1.5, shapeName=dyn_jacket_back_03_cr_capsule, center=<1 0 0>, rotation=<rot 90 0 0>}]
 dyn_jacket_front_01_lf [{shapeType=Capsule, length=5, radius=1.5, shapeName=dyn_jacket_front_01_lf_capsule, center=<1 0 0>, rotation=<rot 90 0 0>}]
 dyn_jacket_front_01_rt [{shapeType=Capsule, length=5, radius=1.5, shapeName=dyn_jacket_front_01_rt_capsule, center=<1 0 0>, rotation=<rot 90 0 0>}]
 dyn_jacket_front_02_lf [{shapeType=Capsule, length=5, radius=1.5, shapeName=dyn_jacket_front_02_lf_capsule, center=<1 0 0>, rotation=<rot 90 0 0>}]
 dyn_jacket_front_02_rt [
  {shapeType=Capsule, length=5, radius=1.5, shapeName=dyn_jacket_front_02_rt_capsule, center=<1 0 0>, rotation=<rot 90 0 0>}
  {
   shapeType=Box
   extentX=7.5
   extentY=2.2914
   extentZ=7.5
   shapeName=dyn_jacket_front_02_rt_box
   center=<6.2708 0.1942 0.0234>
   rotation=<rot 11.6358 0 -7.0274>
  }
 ]
 dyn_jacket_front_03_lf [
  {shapeType=Capsule, length=5, radius=1.5, shapeName=dyn_jacket_front_03_lf_capsule, center=<-0.9427 0 0.6>, rotation=<rot -90 0 0>}
  {
   shapeType=Capsule
   length=7
   radius=2
   shapeName=dyn_jacket_front_03_lf_capsule
   center=<-1.7864 0.7802 -0.5244>
   rotation=<rot -16.1648 -0.0001 0>
  }
 ]
 dyn_jacket_front_03_rt [{
   shapeType=Box
   extentX=4.7245
   extentY=2.4865
   extentZ=9.5173
   shapeName=dyn_jacket_front_03_rt_box
   center=<-0.3109 1.1811 1.2673>
   rotation=<rot 20.9336 -0.0182 -7.7195>
  }]
 dyn_jacket_side_back_01_lf [{shapeType=Capsule, length=5, radius=1.5, shapeName=dyn_jacket_side_back_01_lf_capsule, center=<1 0 0>, rotation=<rot 90 0 0>}]
 dyn_jacket_side_back_01_rt [{shapeType=Capsule, length=5, radius=1.5, shapeName=dyn_jacket_side_back_01_rt_capsule, center=<1 0 0>, rotation=<rot 90 0 0>}]
 dyn_jacket_side_back_02_lf [{shapeType=Capsule, length=5, radius=1.5, shapeName=dyn_jacket_side_back_02_lf_capsule, rotation=<rot 90 0 0>}]
 dyn_jacket_side_back_02_rt [{shapeType=Capsule, length=5, radius=1.5, shapeName=dyn_jacket_side_back_02_rt_capsule, rotation=<rot 90 0 0>}]
 dyn_jacket_side_back_03_lf [{shapeType=Capsule, length=5, radius=1.5, shapeName=dyn_jacket_side_back_03_lf_capsule, center=<1 0 0>, rotation=<rot 90 0 0>}]
 dyn_jacket_side_back_03_rt [{shapeType=Capsule, length=5, radius=1.5, shapeName=dyn_jacket_side_back_03_rt_capsule, center=<1 0 0>, rotation=<rot 90 0 0>}]
 dyn_jacket_side_front_01_lf [{shapeType=Capsule, length=5, radius=1.5, shapeName=dyn_jacket_side_front_01_lf_capsule, center=<1 0 0>, rotation=<rot 90 0 0>}]
 dyn_jacket_side_front_01_rt [{shapeType=Capsule, length=5, radius=1.5, shapeName=dyn_jacket_side_front_01_rt_capsule, center=<1 0 0>, rotation=<rot 90 0 0>}]
 dyn_jacket_side_front_02_lf [{shapeType=Capsule, length=5, radius=1.5, shapeName=dyn_jacket_side_front_02_lf_capsule, rotation=<rot 90 0 0>}]
 dyn_jacket_side_front_02_rt [{shapeType=Capsule, length=5, radius=1.5, shapeName=dyn_jacket_side_front_02_rt_capsule, rotation=<rot 90 0 0>}]
 dyn_jacket_side_front_03_lf [{shapeType=Capsule, length=5, radius=1.5, shapeName=dyn_jacket_side_front_03_lf_capsule, center=<1 0 0>, rotation=<rot 90 0 0>}]
 dyn_jacket_side_front_03_rt [{
   shapeType=Box
   extentX=7.5
   extentY=2.1963
   extentZ=12.6669
   shapeName=dyn_jacket_side_front_03_rt_box
   center=<-3.289 0.1028 -0.4664>
   rotation=<rot 8.5074 -8.9068 -16.0518>
  }]
 pelvis [{shapeType=Capsule, length=10.5588, radius=10.4968, rotation=<rot 0 -90 0>}]
 spine_01 [{shapeType=Capsule, length=6.2335, radius=9.8906, shapeName=spine_01_capsule, rotation=<rot 0 -90.0001 0>}]
 thigh_l [{shapeType=Capsule, length=22.3657, radius=8.668, center=<-21.2327 0.8067 0.8284>, rotation=<rot 0 -90.0851 -88.9694>}]
 thigh_r [
  {shapeType=Capsule, length=22.3657, radius=8.668, center=<21.2396 -0.2442 -1.0096>, rotation=<rot -0.0476 -87.439 91.0295>}
  {shapeType=Capsule, length=16.224, radius=5.776, shapeName=thigh_r_capsule, center=<22.7037 -1.165 9.3086>, rotation=<rot -85 180 180>}
  {shapeType=Capsule, length=3, radius=4, shapeName=thigh_r_capsule, center=<12.901 4.489 10.5804>, rotation=<rot 0 11.8337 90.026>}
 ]
}
constraints [
 {bone1Name=thigh_l, bone2Name=pelvis, swing1LimitDegrees=45, swing2LimitDegrees=45, twistLimitDegrees=45}
 {bone1Name=thigh_r, bone2Name=pelvis, swing1LimitDegrees=45, swing2LimitDegrees=45, twistLimitDegrees=45}
 {
  bone1Name=base_jacket_back_cr
  bone2Name=spine_01
  swing1LimitDegrees=45
  swing1Motion=Free
  swing2LimitDegrees=45
  swing2Motion=Free
  twistLimitDegrees=45
  twistMotion=Free
 }
 {
  bone1Name=dyn_jacket_back_01_cr
  bone2Name=base_jacket_back_cr
  swing1LimitDegrees=5
  swing2LimitDegrees=0
  twistLimitDegrees=45
  twistMotion=Locked
 }
 {
  bone1Name=dyn_jacket_back_02_cr
  bone2Name=dyn_jacket_back_01_cr
  swing1LimitDegrees=10
  swing2LimitDegrees=0
  twistLimitDegrees=45
  twistMotion=Locked
 }
 {
  bone1Name=dyn_jacket_back_03_cr
  bone2Name=dyn_jacket_back_02_cr
  swing1LimitDegrees=20
  swing2LimitDegrees=15
  twistLimitDegrees=45
  twistMotion=Locked
 }
 {
  bone1Name=base_jacket_front_lf
  bone2Name=spine_01
  swing1LimitDegrees=45
  swing1Motion=Free
  swing2LimitDegrees=45
  swing2Motion=Free
  twistLimitDegrees=45
  twistMotion=Free
 }
 {
  bone1Name=dyn_jacket_front_01_lf
  bone2Name=base_jacket_front_lf
  swing1LimitDegrees=5
  swing2LimitDegrees=0
  twistLimitDegrees=45
  twistMotion=Locked
 }
 {
  bone1Name=dyn_jacket_front_02_lf
  bone2Name=dyn_jacket_front_01_lf
  swing1LimitDegrees=20
  swing2LimitDegrees=10
  twistLimitDegrees=45
  twistMotion=Locked
 }
 {
  bone1Name=dyn_jacket_front_03_lf
  bone2Name=dyn_jacket_front_02_lf
  swing1LimitDegrees=20
  swing2LimitDegrees=10
  twistLimitDegrees=45
  twistMotion=Locked
 }
 {
  bone1Name=base_jacket_front_rt
  bone2Name=spine_01
  swing1LimitDegrees=45
  swing1Motion=Free
  swing2LimitDegrees=45
  swing2Motion=Free
  twistLimitDegrees=45
  twistMotion=Free
 }
 {
  bone1Name=dyn_jacket_front_01_rt
  bone2Name=base_jacket_front_rt
  swing1LimitDegrees=5
  swing2LimitDegrees=0
  twistLimitDegrees=45
  twistMotion=Locked
 }
 {
  bone1Name=dyn_jacket_front_02_rt
  bone2Name=dyn_jacket_front_01_rt
  swing1LimitDegrees=15
  swing2LimitDegrees=0
  twistLimitDegrees=45
  twistMotion=Locked
 }
 {
  bone1Name=dyn_jacket_front_03_rt
  bone2Name=dyn_jacket_front_02_rt
  swing1LimitDegrees=20
  swing2LimitDegrees=10
  twistLimitDegrees=45
  twistMotion=Locked
 }
 {
  bone1Name=base_jacket_side_back_lf
  bone2Name=spine_01
  swing1LimitDegrees=45
  swing1Motion=Free
  swing2LimitDegrees=45
  swing2Motion=Free
  twistLimitDegrees=45
  twistMotion=Free
 }
 {
  bone1Name=dyn_jacket_side_back_01_lf
  bone2Name=base_jacket_side_back_lf
  swing1LimitDegrees=5
  swing2LimitDegrees=0
  twistLimitDegrees=45
  twistMotion=Locked
 }
 {
  bone1Name=dyn_jacket_side_back_02_lf
  bone2Name=dyn_jacket_side_back_01_lf
  swing1LimitDegrees=5
  swing2LimitDegrees=0
  twistLimitDegrees=45
  twistMotion=Locked
 }
 {
  bone1Name=dyn_jacket_side_back_03_lf
  bone2Name=dyn_jacket_side_back_02_lf
  swing1LimitDegrees=20
  swing2LimitDegrees=10
  twistLimitDegrees=45
  twistMotion=Locked
 }
 {
  bone1Name=base_jacket_side_back_rt
  bone2Name=spine_01
  swing1LimitDegrees=45
  swing1Motion=Free
  swing2LimitDegrees=45
  swing2Motion=Free
  twistLimitDegrees=45
  twistMotion=Free
 }
 {
  bone1Name=dyn_jacket_side_back_01_rt
  bone2Name=base_jacket_side_back_rt
  swing1LimitDegrees=5
  swing2LimitDegrees=0
  twistLimitDegrees=45
  twistMotion=Locked
 }
 {
  bone1Name=dyn_jacket_side_back_02_rt
  bone2Name=dyn_jacket_side_back_01_rt
  swing1LimitDegrees=5
  swing2LimitDegrees=0
  twistLimitDegrees=45
  twistMotion=Locked
 }
 {
  bone1Name=dyn_jacket_side_back_03_rt
  bone2Name=dyn_jacket_side_back_02_rt
  swing1LimitDegrees=20
  swing2LimitDegrees=10
  twistLimitDegrees=45
  twistMotion=Locked
 }
 {
  bone1Name=base_jacket_side_front_lf
  bone2Name=spine_01
  swing1LimitDegrees=45
  swing1Motion=Free
  swing2LimitDegrees=45
  swing2Motion=Free
  twistLimitDegrees=45
  twistMotion=Free
 }
 {
  bone1Name=dyn_jacket_side_front_01_lf
  bone2Name=base_jacket_side_front_lf
  swing1LimitDegrees=10
  swing2LimitDegrees=0
  twistLimitDegrees=45
  twistMotion=Locked
 }
 {
  bone1Name=dyn_jacket_side_front_02_lf
  bone2Name=dyn_jacket_side_front_01_lf
  swing1LimitDegrees=15
  swing2LimitDegrees=0
  twistLimitDegrees=45
  twistMotion=Locked
 }
 {
  bone1Name=dyn_jacket_side_front_03_lf
  bone2Name=dyn_jacket_side_front_02_lf
  swing1LimitDegrees=20
  swing2LimitDegrees=10
  twistLimitDegrees=45
  twistMotion=Locked
 }
 {
  bone1Name=base_jacket_side_front_rt
  bone2Name=spine_01
  swing1LimitDegrees=45
  swing1Motion=Free
  swing2LimitDegrees=45
  swing2Motion=Free
  twistLimitDegrees=45
  twistMotion=Free
 }
 {
  bone1Name=dyn_jacket_side_front_01_rt
  bone2Name=base_jacket_side_front_rt
  swing1LimitDegrees=10
  swing2LimitDegrees=0
  twistLimitDegrees=45
  twistMotion=Locked
 }
 {
  bone1Name=dyn_jacket_side_front_02_rt
  bone2Name=dyn_jacket_side_front_01_rt
  swing1LimitDegrees=15
  swing2LimitDegrees=0
  twistLimitDegrees=45
  twistMotion=Locked
 }
 {
  bone1Name=dyn_jacket_side_front_03_rt
  bone2Name=dyn_jacket_side_front_02_rt
  swing1LimitDegrees=20
  swing2LimitDegrees=10
  twistLimitDegrees=45
  twistMotion=Locked
 }
 {
  bone1Name=spine_01
  bone2Name=pelvis
  swing1LimitDegrees=45
  swing1Motion=Free
  swing2LimitDegrees=45
  swing2Motion=Free
  twistLimitDegrees=45
  twistMotion=Free
 }
]
