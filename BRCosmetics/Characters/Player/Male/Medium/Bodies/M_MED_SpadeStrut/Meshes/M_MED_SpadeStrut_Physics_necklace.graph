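# Physics asset for the necklace on the SpadeStrut male medium player body cosmetic, simulating its upper and lower strands.
# Small base capsules for each strand link to a large spine_05 torso capsule through constraints with free swing and twist.
# The upper strand has one dynamic capsule and the lower chains three, with tight per-link limits that allow no twist and favor swing on one axis.
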PhysicsAsset /BRCosmetics/Characters/Player/Male/Medium/Bodies/M_MED_SpadeStrut/Meshes/M_MED_SpadeStrut_Physics_necklace
bodies {
 base_necklace_lower_01_cr [{shapeType=Capsule, length=0.505, radius=0.505, shapeName=base_necklace_lower_01_cr_capsule, rotation=<rot 0 -90 0>}]
 base_necklace_upper_01_cr [{shapeType=Capsule, length=0.505, radius=0.505, shapeName=base_necklace_upper_01_cr_capsule, rotation=<rot 0 -90 0>}]
 dyn_necklace_lower_01_cr [{shapeType=Capsule, length=5, radius=2, shapeName=dyn_necklace_lower_01_cr_capsule, center=<1 0 6>, rotation=<rot 0 0 90>}]
 dyn_necklace_lower_02_cr [{shapeType=Capsule, length=5, radius=1.5, shapeName=dyn_necklace_lower_02_cr_capsule, center=<1 0 1.8>, rotation=<rot 0 0 90>}]
 dyn_necklace_lower_03_cr [{shapeType=Capsule, length=5, radius=1.5, shapeName=dyn_necklace_lower_03_cr_capsule, center=<1.6722 0 1.9>, rotation=<rot 0 0 90>}]
 dyn_necklace_upper_01_cr [{shapeType=Capsule, length=3, radius=1, shapeName=dyn_necklace_upper_01_cr_capsule, center=<9.2215 0 2>, rotation=<rot 0 0 90>}]
 spine_05 [{shapeType=Capsule, length=12.0621, radius=9, shapeName=spine_04_capsule, center=<-0.6264 2.8239 -0.2339>, rotation=<rot 90 -21.1445 0>}]
}
constraints [
 {bone1Name=dyn_necklace_upper_01_cr, bone2Name=base_necklace_upper_01_cr, swing1LimitDegrees=0, swing2LimitDegrees=10, twistLimitDegrees=0}
 {bone1Name=dyn_necklace_lower_03_cr, bone2Name=dyn_necklace_lower_02_cr, swing1LimitDegrees=3, swing2LimitDegrees=40, twistLimitDegrees=0}
 {bone1Name=dyn_necklace_lower_02_cr, bone2Name=dyn_necklace_lower_01_cr, swing1LimitDegrees=5, swing2LimitDegrees=20, twistLimitDegrees=0}
 {bone1Name=dyn_necklace_lower_01_cr, bone2Name=base_necklace_lower_01_cr, swing1LimitDegrees=3, swing2LimitDegrees=15, twistLimitDegrees=0}
 {
  bone1Name=base_necklace_upper_01_cr
  bone2Name=spine_05
  swing1LimitDegrees=45
  swing1Motion=Free
  swing2LimitDegrees=45
  swing2Motion=Free
  twistLimitDegrees=45
  twistMotion=Free
 }
 {
  bone1Name=base_necklace_lower_01_cr
  bone2Name=spine_05
  swing1LimitDegrees=45
  swing1Motion=Free
  swing2LimitDegrees=45
  swing2Motion=Free
  twistLimitDegrees=45
  twistMotion=Free
 }
]
