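# Post-process style AnimBP for the M_MED_SpadeStrut cosmetic body, adding secondary motion to the pose copied from its attached parent mesh.
# A crouch-aware Control Rig feeds body, necklace and jacket RigidBody sims, a weighted gun-bone constraint, then AnimDynamics for strap, zippers and handle, linked by cached poses.
# The update event sets gravity and RBAN settings via a parameter asset, plus falling and front-end values.
# The left-hand handle constraint is unconnected.
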AnimBlueprint /BRCosmetics/Characters/Player/Male/Medium/Bodies/M_MED_SpadeStrut/Meshes/M_MED_SpadeStrut_AnimBP
target_skeleton=&/M_MED_SpadeStrut_Skeleton
graphs [
 AnimGraph {nodes [
   "Constraint - base_handle_hand_lf (PPPPPPPP)"!(
    ComponentPose
    Alpha=1
    ConstraintWeights_0=0.2302
    ConstraintWeights_1:float=0.2208
    ConstraintWeights_2:float=0.204
    ConstraintWeights_3:float=0.1297
    ConstraintWeights_4:float=0.0761
    ConstraintWeights_5:float=0.061
    ConstraintWeights_6:float=0.0463
    ConstraintWeights_7:float=0.0318
   ) -> (Pose) {
    boneToModify {boneName=base_handle_hand_lf}
    constraintSetup [
     {targetBone {boneName=lowerarm_twist_01_l}}
     {targetBone {boneName=thumb_01_l}}
     {targetBone {boneName=pinky_metacarpal_l}}
     {targetBone {boneName=hand_l}}
     {targetBone {boneName=lowerarm_twist_02_l}}
     {targetBone {boneName=ring_metacarpal_l}}
     {targetBone {boneName=thumb_02_l}}
     {targetBone {boneName=index_metacarpal_l}}
    ]
    constraintWeights [1,1,1,1,1,1,1,1]
   }
   Output Pose(Result=Use cached pose 'dyn_cache'()) {layerGroup=None}
   n1 = "Constraint - base_gun_rt (PPPPPPPP)"!(
    ComponentPose=Local To Component(LocalPose=Use cached pose 'control_rig_cache'())
    Alpha=1
    ConstraintWeights_0=0.6986
    ConstraintWeights_1:float=0.28
    ConstraintWeights_2:float=0.0102
    ConstraintWeights_3:float=0.0043
    ConstraintWeights_4:float=0.0033
    ConstraintWeights_5:float=0.0016
    ConstraintWeights_6:float=0.0012
    ConstraintWeights_7:float=0.0008
   ) -> (Pose) {
    boneToModify {boneName=base_gun_rt}
    constraintSetup [
     {targetBone {boneName=thigh_twist_01_r}}
     {targetBone {boneName=thigh_r}}
     {targetBone {boneName=pelvis}}
     {targetBone {boneName=deform_groin_r}}
     {targetBone {boneName=calf_r}}
     {targetBone {boneName=deform_glute_r}}
     {targetBone {boneName=deform_knee_out_r}}
     {targetBone {boneName=deform_knee_in_r}}
    ]
    constraintWeights [1,1,1,1,1,1,1,1]
   }
   Save cached pose 'predyn_cache'(Pose=Component To Local(ComponentPose=@n1)) {cacheName=predyn_cache}
   n2 = Anim Dynamics!(ComponentPose=Local To Component(LocalPose=Use cached pose 'predyn_cache'()), Alpha=1, GravityOverride:Vector="0,0,-980") -> (Pose) {
    angularDampingOverride=1
    angularSpringConstant=2000
    bAngularSpring=true
    bOverrideAngularDamping=true
    bUseGravityOverride=true
    gravityScale=0.35
    simulationSpace=RootRelative
    alphaBoolBlend {blendOption=Linear}
    alphaScaleBiasClamp {interpSpeedDecreasing=2.5, interpSpeedIncreasing=8}
    componentAppliedLinearAccClamp=<3000 3000 3000>
    componentLinearAccScale=<0.2 0.2 0.2>
    componentLinearVelScale=<0.2 0.2 0.2>
    gravityOverride {z=-980}
    physicsBodyDefinitions [{
      boundBone {boneName=None}
      boxExtents=<20 20 20>
      constraintSetup {angularLimitsMax=<0 0 0>, angularLimitsMin=<0 -30 0>, angularTarget=<1 0 0>}
      localJointOffset=<5 0 0>
     }]
    relativeSpaceBone {boneName=root}
    simSpaceSettings {maxAngularAcceleration=35, maxAngularVelocity=7}
    boundBone=~
   }
   n3 = Anim Dynamics - Bone: dyn_strap_01_cr!(ComponentPose=@n2, Alpha=1, GravityOverride:Vector="0,0,-980") -> (Pose) {
    angularDampingOverride=1
    angularSpringConstant=700
    bAngularSpring=true
    bOverrideAngularDamping=true
    bUseGravityOverride=true
    gravityScale=0.35
    simulationSpace=BoneRelative
    alphaBoolBlend {blendOption=Linear}
    alphaScaleBiasClamp {interpSpeedDecreasing=2.5, interpSpeedIncreasing=8}
    boundBone {boneName=dyn_strap_01_cr}
    componentAppliedLinearAccClamp=<3000 3000 3000>
    componentLinearAccScale=<0.3 0.3 0.3>
    componentLinearVelScale=<0.2 0.2 0.2>
    gravityOverride {z=-980}
    physicsBodyDefinitions [{
      boundBone {boneName=dyn_strap_01_cr}
      boxExtents=<20 20 20>
      constraintSetup {angularLimitsMax=<0 0 3>, angularLimitsMin=<0 -20 -3>, angularTarget=<1 0 0>}
      localJointOffset=<5 0 0>
     }]
    simSpaceSettings {maxAngularAcceleration=40, maxAngularVelocity=8, simSpaceAngularAlpha=1}
   }
   n4 = Anim Dynamics - Bone: dyn_zipper_bag_side_lf!(ComponentPose=@n3, Alpha=1, GravityOverride:Vector="0,0,-980") -> (Pose) {
    angularDampingOverride=1
    angularSpringConstant=1000
    bAngularSpring=true
    bOverrideAngularDamping=true
    bUseGravityOverride=true
    gravityScale=0.35
    simulationSpace=RootRelative
    alphaBoolBlend {blendOption=Linear}
    alphaScaleBiasClamp {interpSpeedDecreasing=2.5, interpSpeedIncreasing=8}
    boundBone {boneName=dyn_zipper_bag_side_lf}
    componentAppliedLinearAccClamp=<3000 3000 3000>
    componentLinearAccScale=<0.3 0.3 0.3>
    componentLinearVelScale=<0.3 0.3 0.3>
    gravityOverride {z=-980}
    physicsBodyDefinitions [{
      boundBone {boneName=dyn_zipper_bag_side_lf}
      boxExtents=<10 10 10>
      constraintSetup {angularLimitsMax=<0 25 0>, angularLimitsMin=<0 -25 -20>, angularTarget=<1 0 0>}
      localJointOffset=<6 0 0>
     }]
    relativeSpaceBone {boneName=root}
    simSpaceSettings {maxAngularAcceleration=35, maxAngularVelocity=7, simSpaceAngularAlpha=1}
   }
   n5 = Anim Dynamics - Bone: dyn_zipper_bag_front_cr!(ComponentPose=@n4, Alpha=1, GravityOverride:Vector="0,0,-980") -> (Pose) {
    angularDampingOverride=1
    angularSpringConstant=1000
    bAngularSpring=true
    bOverrideAngularDamping=true
    bUseGravityOverride=true
    gravityScale=0.35
    simulationSpace=RootRelative
    alphaBoolBlend {blendOption=Linear}
    alphaScaleBiasClamp {interpSpeedDecreasing=2.5, interpSpeedIncreasing=8}
    boundBone {boneName=dyn_zipper_bag_front_cr}
    componentAppliedLinearAccClamp=<3000 3000 3000>
    componentLinearAccScale=<0.3 0.3 0.3>
    componentLinearVelScale=<0.3 0.3 0.3>
    gravityOverride {z=-980}
    physicsBodyDefinitions [{
      boundBone {boneName=dyn_zipper_bag_front_cr}
      boxExtents=<10 10 10>
      constraintSetup {angularLimitsMax=<0 0 15>, angularLimitsMin=<0 -15 -15>, angularTarget=<1 0 0>}
      localJointOffset=<6 0 0>
     }]
    relativeSpaceBone {boneName=root}
    simSpaceSettings {maxAngularAcceleration=35, maxAngularVelocity=7, simSpaceAngularAlpha=1}
   }
   n6 = Anim Dynamics - Bone: dyn_handle_hand_lf!(ComponentPose=@n5, Alpha=1, GravityOverride:Vector="0,0,-980") -> (Pose) {
    angularDampingOverride=1
    angularSpringConstant=700
    bAngularSpring=true
    bOverrideAngularDamping=true
    bUseGravityOverride=true
    gravityScale=0.35
    alphaBoolBlend {blendOption=Linear}
    alphaScaleBiasClamp {interpSpeedDecreasing=2.5, interpSpeedIncreasing=8}
    boundBone {boneName=dyn_handle_hand_lf}
    componentAppliedLinearAccClamp=<3000 3000 3000>
    componentLinearAccScale=<0.3 0.3 0.3>
    componentLinearVelScale=<0.3 0.3 0.3>
    gravityOverride {z=-980}
    physicsBodyDefinitions [{
      boundBone {boneName=dyn_handle_hand_lf}
      boxExtents=<25 25 25>
      constraintSetup {angularLimitsMax=<0 3 0>, angularLimitsMin=<0 -3 0>, angularTarget=<1 0 0>}
      localJointOffset=<4 0 0>
     }]
    relativeSpaceBone {boneName=root}
    simSpaceSettings {maxAngularAcceleration=35, maxAngularVelocity=7, simSpaceAngularAlpha=1}
   }
   Save cached pose 'dyn_cache'(Pose=Component To Local(ComponentPose=@n6)) {cacheName=dyn_cache}
   n7 = Control Rig!(Source=Copy Pose From Mesh() {bUseAttachedParent=true}, Alpha, IsCrouching:bool=Get bIsCrouching()) -> (Pose) {controlRigAssetReference {blueprintRigClass=&/M_MED_SpadeStrut_CtrlRig.M_MED_SpadeStrut_CtrlRig_C}}
   n8 = RigidBody!(ComponentPose=Local To Component(LocalPose=@n7), Alpha, OverrideWorldGravity:Vector="0,0,-980") -> (Pose) {
    bClampLinearTranslationLimitToRefPose=true
    bOverrideWorldGravity=true
    overridePhysicsAsset=&/M_MED_SpadeStrut_Physics
    simulationSpace=BaseBoneSpace
    alphaBoolBlend {blendOption=Linear}
    baseBoneRef {boneName=pelvis}
    overrideWorldGravity {z=-980}
    simSpaceSettings {
     dampingAlpha=0
     maxAngularAcceleration=45
     maxAngularVelocity=8
     maxLinearAcceleration=3500
     maxLinearVelocity=700
     velocityScaleZ=0.5
     worldAlpha=0.8
     externalAngularVelocity=<0.2 0.2 0.2>
     externalLinearDragV=<0.3 0.3 0.3>
     externalLinearVelocity=<0.2 0.2 0.2>
    }
   }
   n9 = RigidBody(ComponentPose=@n8, OverrideWorldGravity="0,0,0") {
    bClampLinearTranslationLimitToRefPose=true
    bOverrideWorldGravity=true
    overridePhysicsAsset=&/M_MED_SpadeStrut_Physics_necklace
    simulationSpace=BaseBoneSpace
    baseBoneRef {boneName=spine_05}
    componentAppliedLinearAccClamp=<3000 3000 3000>
    simSpaceSettings {
     dampingAlpha=0.1
     maxAngularAcceleration=40
     maxAngularVelocity=8
     maxLinearAcceleration=3500
     maxLinearVelocity=700
     velocityScaleZ=0.5
     worldAlpha=1
     externalAngularVelocity=<0.2 0.2 0.2>
     externalLinearVelocity=<0.2 0.2 0.2>
    }
   }
   n10 = RigidBody(ComponentPose=@n9, OverrideWorldGravity="0,0,0") {
    bClampLinearTranslationLimitToRefPose=true
    bOverrideWorldGravity=true
    overridePhysicsAsset=&/M_MED_SpadeStrut_Physics_jacket
    simulationSpace=BaseBoneSpace
    baseBoneRef {boneName=pelvis}
    simSpaceSettings {worldAlpha=0.35}
   }
   Save cached pose 'control_rig_cache'(Pose=Component To Local(ComponentPose=@n10)) {cacheName=control_rig_cache}
  ]}
 EventGraph {nodes [
   Event BlueprintUpdateAnimation()
   > n1 = Sequence()
   > n2 = "Gravity Override State Function (Parameter Asset) (Nativized)"(
    CharacterPartAnimInstance=Self-Reference()
    GameTimeInSeconds=GetGameTimeInSeconds()
    Parameters=&/M_MED_SpadeStrut_Parameters
   )
   > Set GravityOverride(GravityOverride=@n2)
   > Set RBAN Settings Out(RBAN Settings Out=@n2.OutRigidBodyAnimNodeParameters)
   > Branch(Condition=Get bIsFalling())
   > Set GravityOverride_cloth(GravityOverride_cloth:Vector="0,-500,-980") -> (Output_Get:Vector)
   n3 = Branch(execute=@n1.then_1, Condition=NOT Boolean(A=Get bIsInFrontEnd()))
   > Set FrontEndAlpha(FrontEndAlpha:double=0.05) -> (Output_Get:double)
   > Set spacesettings(spacesettings:Byte=4) -> (Output_Get:Byte=0)
   > Set GravityOverride_cloth(GravityOverride_cloth="0,0,-1200")
   Set FrontEndAlpha(execute=@n3.else, FrontEndAlpha=1)
   > Set spacesettings(spacesettings=3) -> (Output_Get=0)
   > Set GravityOverride_cloth(GravityOverride_cloth="0,0,-980")
  ]}
]
variables [
 FrontEndAlpha {type=number}
 GravityOverride {type=Vector}
 GravityOverride_cloth {type=Vector}
 RBAN Settings Out {type=FortRigidBodyAnimNodeParameters}
 RbanAlpha {type=number}
 RbanFEAlpha {type=number}
 spacesettings {type=integer}
]
